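; Looping gameplay cue that shows a life-drain impact effect on a Fortnite player pawn while a non-participant boss, taken from the cue's source object, drains it.
; On looping start it attaches the Niagara loop to the target's spine_05 socket and aims its rotation and SparkDirection at the boss's FX_Snap socket.
; A repeating timer refreshes that aim.
; Removal clears the timer and spawns a disconnect effect at the last rotation.
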(event EventOnLoopingStart(EitherVFX-preferthisversion)
  (params (MyTarget :type Actor%20Object%20Reference) (Parameters :type Gameplay%20Cue%20Parameters%20Structure%20%28by%20ref%29) (ParticleComponents :type Array%20of%20FXSystem%20Component%20Object%20References) (AudioComponents :type Array%20of%20Audio%20Component%20Object%20References))
  (bind (_normalizedmagnitude _rawmagnitude _effectcontext _matchedtagname _originaltag _aggregatedsourcetags _aggregatedtargettags _location _normal _instigator _effectcauser _sourceobject) (Ability|GameplayCue|BreakGameplayCueParameters Parameters))
  (bind _as_fort_player_pawn (Variables|Default|GetAs%20Fort%20Player%20Pawn))
  (bind _mesh (Variables|/Script/Engine.Character|GetMesh _as_fort_player_pawn))
  (bind _returnvalue (Transformation|GetSocketLocation _mesh "spine_05"))
  (bind _ns_lifedrain_impact_loop (Variables|Default|GetNS_LifeDrain_Impact_Loop))
  (bind _returnvalue_2 (Utilities|Time|SetTimerbyEvent (delegate Custom|CE_Update%20Impact%20Rotation) 0.1 true))
  (Variables|Default|SetTimer _returnvalue_2)
  (Utilities|FlowControl|Sequence
    (:then_0
      (bind _asfort_player_pawn (Utilities|Casting|CastToFortPlayerPawn MyTarget))
      (bind _output_get (Variables|Default|SetAs%20Fort%20Player%20Pawn _asfort_player_pawn))
      (Transformation|AttachComponentToComponent (Variables|Default|GetNS_LifeDrain_Impact_Loop) (Variables|/Script/Engine.Character|GetMesh _output_get) "spine_05" "KeepRelative" "KeepRelative" "KeepRelative" true))
    (:then_1
      (bind _asbp_player_pawn_non_participant_boss (Utilities|Casting|CastToBP_PlayerPawn_NonParticipant_Boss _sourceobject))
      (bind _output_get_2 (Variables|Default|SetAs%20BP%20Player%20Pawn%20Non%20Participant%20Boss _asbp_player_pawn_non_participant_boss)))
    (:then_2
      (bind _output_get_3 (Variables|Default|SetRotation (Math|Rotator|FindLookatRotation _returnvalue (Transformation|GetSocketLocation (Variables|/Script/Engine.Character|GetMesh _output_get_2) "FX_Snap"))))
      (Transformation|SetWorldRotation _ns_lifedrain_impact_loop _output_get_3 false false)
      (Niagara|SetNiagaraVariable(Vector3) _ns_lifedrain_impact_loop "SparkDirection" (Math|Vector|GetUnitDirection(Vector) _returnvalue (Transformation|GetSocketLocation (Variables|/Script/Engine.Character|GetMesh _output_get_2) "FX_Snap"))))))

(event Custom|CE_Update%20Impact%20Rotation
  (bind _as_fort_player_pawn (Variables|Default|GetAs%20Fort%20Player%20Pawn))
  (bind _mesh (Variables|/Script/Engine.Character|GetMesh _as_fort_player_pawn))
  (bind _returnvalue (Transformation|GetSocketLocation _mesh "spine_05"))
  (bind _ns_lifedrain_impact_loop (Variables|Default|GetNS_LifeDrain_Impact_Loop))
  (bind _output_get (Variables|Default|SetRotation (Math|Rotator|FindLookatRotation _returnvalue (Transformation|GetSocketLocation (Variables|/Script/Engine.Character|GetMesh (stale EventOnLoopingStart(EitherVFX-preferthisversion) _output_get_2)) "FX_Snap"))))
  (Transformation|SetWorldRotation _ns_lifedrain_impact_loop _output_get false false)
  (Niagara|SetNiagaraVariable(Vector3) _ns_lifedrain_impact_loop "SparkDirection" (Math|Vector|GetUnitDirection(Vector) _returnvalue (Transformation|GetSocketLocation (Variables|/Script/Engine.Character|GetMesh (stale EventOnLoopingStart(EitherVFX-preferthisversion) _output_get_2)) "FX_Snap"))))

(event EventOnRemoval(EitherVFX-preferthisversion)
  (params (MyTarget :type Actor%20Object%20Reference) (Parameters :type Gameplay%20Cue%20Parameters%20Structure%20%28by%20ref%29) (ParticleComponents :type Array%20of%20FXSystem%20Component%20Object%20References) (AudioComponents :type Array%20of%20Audio%20Component%20Object%20References) (BurstCameraShakeInstance :type Camera%20Shake%20Base%20Object%20Reference) (BurstDecalInstance :type Decal%20Actor%20Object%20Reference))
  (Utilities|FlowControl|Sequence
    (:then_0
      (Utilities|Time|ClearandInvalidateTimerbyHandle 0 (Variables|Default|GetTimer)))
    (:then_1
      (bind _returnvalue (Niagara|SpawnSystemAttached "/NPCLibrary/Abilities/LifeDrain/VFX/Target/NS_LifeDrain_Impact_Disconnect.NS_LifeDrain_Impact_Disconnect" (Variables|/Script/Engine.Character|GetMesh (Variables|Default|GetAs%20Fort%20Player%20Pawn)) "spine_05" "0, 0, 0" "0, 0, 0" "KeepRelativeOffset" true))
      (Transformation|SetWorldRotation _returnvalue (Variables|Default|GetRotation) false false))))

(fn ConstructionScript)
