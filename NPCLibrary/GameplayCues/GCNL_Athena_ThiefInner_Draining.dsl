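; Looping gameplay cue notify for the Thief Inner life-drain effect, showing energy VFX on the affected player pawn.
; On looping start it caches the Fort player pawn, waits a zero-length delay, then attaches the loop Niagara system to the mesh's FX_Snap socket and stores it.
; On removal, a check on the stored loop component gates deactivating it and spawning an end burst at the same socket.
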(event EventBeginPlay)

(event Collision|EventActorBeginOverlap
  (params (OtherActor :type Actor%20Object%20Reference)))

(event EventTick
  (params (DeltaSeconds :type Float%20%28single-precision%29)))

(event EventOnLoopingStart(EitherVFX-preferthisversion)
  (params (MyTarget :type Actor%20Object%20Reference) (Parameters :type Gameplay%20Cue%20Parameters%20Structure%20%28by%20ref%29) (ParticleComponents :type Array%20of%20FXSystem%20Component%20Object%20References) (AudioComponents :type Array%20of%20Audio%20Component%20Object%20References))
  (bind _asfort_player_pawn (Utilities|Casting|CastToFortPlayerPawn MyTarget))
  (bind _output_get (Variables|Default|SetAs%20Fort%20Player%20Pawn _asfort_player_pawn))
  (Utilities|FlowControl|Delay 0 0.0)
  (bind _returnvalue (Niagara|SpawnSystemAttached "/NPCLibrary/Abilities/LifeDrain/VFX/Source/NS_LifeDrain_Energy_Loop.NS_LifeDrain_Energy_Loop" (Variables|/Script/Engine.Character|GetMesh _output_get) "FX_Snap" "0, 0, 0" "0, 0, 0" "KeepRelativeOffset" true))
  (Variables|Default|SetFXLoop _returnvalue))

(event EventOnRemoval(EitherVFX-preferthisversion)
  (params (MyTarget :type Actor%20Object%20Reference) (Parameters :type Gameplay%20Cue%20Parameters%20Structure%20%28by%20ref%29) (ParticleComponents :type Array%20of%20FXSystem%20Component%20Object%20References) (AudioComponents :type Array%20of%20Audio%20Component%20Object%20References) (BurstCameraShakeInstance :type Camera%20Shake%20Base%20Object%20Reference) (BurstDecalInstance :type Decal%20Actor%20Object%20Reference))
  (bind _fxloop (Variables|Default|GetFXLoop)
    (:then
      (Components|Activation|Deactivate _fxloop)
      (Niagara|SpawnSystemAttached "/NPCLibrary/Abilities/LifeDrain/VFX/Source/NS_LifeDrain_Energy_End.NS_LifeDrain_Energy_End" (Variables|/Script/Engine.Character|GetMesh (Variables|Default|GetAs%20Fort%20Player%20Pawn)) "FX_Snap" "0, 0, 0" "0, 0, 0" "KeepRelativeOffset" true))
    (:else)))

(fn ConstructionScript)
