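; Looping gameplay cue for the Lockdown ability that shows a link beam on an affected wildlife NPC.
; On loop start it waits briefly, casts the target to NPC_Pawn_Wildlife_Parent and attaches the Niagara link effect to its mesh's chest socket.
; It aims the BeamTarget parameter above the cue's location and raises translucent sort priority.
; On removal it deactivates the stored effect.
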(event EventOnLoopingStart(EitherVFX-preferthisversion)
  (params (MyTarget :type Actor%20Object%20Reference) (Parameters :type Gameplay%20Cue%20Parameters%20Structure%20%28by%20ref%29) (ParticleComponents :type Array%20of%20FXSystem%20Component%20Object%20References) (AudioComponents :type Array%20of%20Audio%20Component%20Object%20References))
  (bind (_normalizedmagnitude _rawmagnitude _effectcontext _matchedtagname _originaltag _aggregatedsourcetags _aggregatedtargettags _location) (Ability|GameplayCue|BreakGameplayCueParameters Parameters))
  (bind _linkfx (Variables|Default|GetLinkFX))
  (Utilities|FlowControl|Delay 0 0.5)
  (bind _asnpc_pawn_wildlife_parent (Utilities|Casting|CastToNPC_Pawn_Wildlife_Parent MyTarget))
  (bind _returnvalue (Effects|SpawnFXSystemAttached :FXSystemAsset "/NPCLibrary/Abilities/Lockdown/VFX/NS_Lockdown_Loop_Link.NS_Lockdown_Loop_Link" :AttachToComponent (Variables|/Script/Engine.Character|GetMesh _asnpc_pawn_wildlife_parent) :AttachPointName "QuadSpine_A_Chest_C" :Scale "1.000000,1.000000,1.000000" :LocationType "SnapToTarget" :bAutoDestroy true :PoolingMethod "None" :bAutoActivate true))
  (Variables|Default|SetLinkFX _returnvalue)
  (Effects|Components|ParticleSystem|SetVectorParameter _linkfx "BeamTarget" (+ _location "0,0,250.000000"))
  (Rendering|SetTranslucentSortPriority _linkfx 10))

(event EventOnRemoval(EitherVFX-preferthisversion)
  (params (MyTarget :type Actor%20Object%20Reference) (Parameters :type Gameplay%20Cue%20Parameters%20Structure%20%28by%20ref%29) (ParticleComponents :type Array%20of%20FXSystem%20Component%20Object%20References) (AudioComponents :type Array%20of%20Audio%20Component%20Object%20References) (BurstCameraShakeInstance :type Camera%20Shake%20Base%20Object%20Reference) (BurstDecalInstance :type Decal%20Actor%20Object%20Reference))
  (bind _linkfx (Variables|Default|GetLinkFX)
    (:then
      (Components|Activation|Deactivate _linkfx))
    (:else)))

(fn ConstructionScript)
