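; Looping gameplay cue for a telekinetic throw's hovering object, which plays the object-raise sound.
; On application it passes its raise tag to PlaySound, which gets the Sound Library subsystem and, if valid, plays that tag for the cue's instigator with the target attach component.
; It contains no VFX logic or explicit sound stop, and its BeginPlay, overlap and Tick events are empty.
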(event EventBeginPlay)

(event Collision|EventActorBeginOverlap
  (params (OtherActor :type Actor%20Object%20Reference)))

(event EventTick
  (params (DeltaSeconds :type Float%20%28single-precision%29)))

(event EventOnApplication(EitherVFX-preferthisone)
  (params (MyTarget :type Actor%20Object%20Reference) (Parameters :type Gameplay%20Cue%20Parameters%20Structure%20%28by%20ref%29) (ParticleComponents :type Array%20of%20FXSystem%20Component%20Object%20References) (AudioComponents :type Array%20of%20Audio%20Component%20Object%20References) (BurstCameraShakeInstance :type Camera%20Shake%20Base%20Object%20Reference) (BurstDecalInstance :type Decal%20Actor%20Object%20Reference))
  (CallFunction|PlaySound MyTarget Parameters (Variables|Default|GetObjectRaiseTag)))

(fn PlaySound
  (params (WorldContext :type Object%20Reference) (Parameters :type Gameplay%20Cue%20Parameters%20Structure%20%28by%20ref%29) (Event%20Name :type Gameplay%20Tag%20Structure))
  (bind (_normalizedmagnitude _rawmagnitude _effectcontext _matchedtagname _originaltag _aggregatedsourcetags _aggregatedtargettags _location _normal _instigator _effectcauser _sourceobject _physicalmaterial _gameplayeffectlevel _abilitylevel _targetattachcomponent) (Ability|GameplayCue|BreakGameplayCueParameters Parameters))
  (bind _returnvalue (WorldSubsystems|GetSoundLibrarySubsystem WorldContext))
  ;[1 1 1] Get Sound library from instigator pawn. The play sound based on hit tag from pawn library
  (Utilities|IsValid _returnvalue
    (:Is%20Valid
      (SoundLibrary|PlaySoundSimple _returnvalue _instigator (Utilities|Struct|MakeSimpleSoundSettings Event%20Name true true true true 1.0 1.0 0.0 0 0 "None" _targetattachcomponent) :!class /Script/SoundLibrary.SoundLibrarySubsystem)
      (return))))

(fn ConstructionScript)
