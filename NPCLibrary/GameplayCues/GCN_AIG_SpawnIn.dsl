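; Burst gameplay cue that plays the AIG spawn-in effect, a Niagara explosion at the target actor's location.
; OnBurst uses the cue's gameplay effect level as an index to select the system, though every listed option is the same NS_Explosion.
; The same level picks a Zeus, Cerberus or Hades color, which is written to the system's LightningColor variable.
; The DestroyVFX custom event is empty.
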(event Custom|DestroyVFX)

(fn OnBurst(EitherVFX-preferthisversion)
  (params (MyTarget :type Actor%20Object%20Reference) (Parameters :type Gameplay%20Cue%20Parameters%20Structure%20%28by%20ref%29) (ParticleComponents :type Array%20of%20FXSystem%20Component%20Object%20References) (AudioComponents :type Array%20of%20Audio%20Component%20Object%20References) (BurstCameraShakeInstance :type Camera%20Shake%20Base%20Object%20Reference) (BurstDecalInstance :type Decal%20Actor%20Object%20Reference))
  (bind (_normalizedmagnitude _rawmagnitude _effectcontext _matchedtagname _originaltag _aggregatedsourcetags _aggregatedtargettags _location _normal _instigator _effectcauser _sourceobject _physicalmaterial _gameplayeffectlevel) (Ability|GameplayCue|BreakGameplayCueParameters Parameters))
  (bind _returnvalue (Niagara|SpawnSystematLocation 0 (Utilities|Select :Index _gameplayeffectlevel :Option%200 "/NPCLibrary/VFX/AIGSpawnIn/NS_Explosion.NS_Explosion" :Option%201 "/NPCLibrary/VFX/AIGSpawnIn/NS_Explosion.NS_Explosion" :Option%202 "/NPCLibrary/VFX/AIGSpawnIn/NS_Explosion.NS_Explosion" :Option%203 "/NPCLibrary/VFX/AIGSpawnIn/NS_Explosion.NS_Explosion" :Option%204 "/NPCLibrary/VFX/AIGSpawnIn/NS_Explosion.NS_Explosion") (Transformation|GetActorLocation MyTarget)))
  (Niagara|SetNiagaraVariable(LinearColor) _returnvalue "LightningColor" (Utilities|Select :Index _gameplayeffectlevel :Option%200 (Variables|Default|GetZeusColor) :Option%201 (Variables|Default|GetCerberusColor) :Option%202 (Variables|Default|GetCerberusColor) :Option%203 (Variables|Default|GetZeusColor) :Option%204 (Variables|Default|GetHadesColor) :Option%205 (Variables|Default|GetZeusColor)))
  (return))
