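; Gameplay cue burst notify that plays the hit sound for a TK throw object impact.
; OnBurst gets the sound library subsystem via the instigator and plays the hit tag sound attached to the target's root component with an offset.
; It then sets "Islocal" and "IsPawn" on the first audio component, based on whether the effect causer is a Fortnite player pawn and locally viewed, otherwise both false.
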(fn OnBurst(EitherVFX-preferthisversion)
  (params (MyTarget :type Actor%20Object%20Reference) (Parameters :type Gameplay%20Cue%20Parameters%20Structure%20%28by%20ref%29) (ParticleComponents :type Array%20of%20FXSystem%20Component%20Object%20References) (AudioComponents :type Array%20of%20Audio%20Component%20Object%20References) (BurstCameraShakeInstance :type Camera%20Shake%20Base%20Object%20Reference) (BurstDecalInstance :type Decal%20Actor%20Object%20Reference))
  (bind (_normalizedmagnitude _rawmagnitude _effectcontext _matchedtagname _originaltag _aggregatedsourcetags _aggregatedtargettags _location _normal _instigator _effectcauser) (Ability|GameplayCue|BreakGameplayCueParameters Parameters))
  (bind (_asfort_player_pawn _bsuccess) (Utilities|Casting|CastToFortPlayerPawn _effectcauser))
  (bind _returnvalue (WorldSubsystems|GetSoundLibrarySubsystem _instigator))
  ;[1 1 1] Get Sound library from instigator pawn. The play sound based on hit tag from pawn library
  (Utilities|IsValid _returnvalue
    (:Is%20Valid
      (bind _returnvalue_2 (SoundLibrary|PlaySoundSimple _returnvalue _instigator (Utilities|Struct|MakeSimpleSoundSettings (Variables|Default|GetHitSoundLibTag) true true true true 1.0 1.0 0.0 0 0 "None" (Variables|/Script/Engine.Actor|GetRootComponent MyTarget) "KeepRelativeOffset" "0.000000,0.000000,250.000000") :!class /Script/SoundLibrary.SoundLibrarySubsystem))
      ;[1 1 1] Check for pawn vs object and if pawn is local
      (if (Utilities|Array|IsValidIndex (.AudioComponents (Utilities|Struct|BreakSoundLibraryPlayResult _returnvalue_2)))
        (Utilities|IsValid _asfort_player_pawn
          (:Is%20Valid
            (Audio|Parameter|SetBooleanParameter (Utilities|Array|Get(acopy) (.AudioComponents (Utilities|Struct|BreakSoundLibraryPlayResult _returnvalue_2))) "Islocal" (Pawn|IsLocallyViewed _asfort_player_pawn) :!class /Script/AudioExtensions.AudioParameterControllerInterface)
            (Audio|Parameter|SetBooleanParameter (Utilities|Array|Get(acopy) (.AudioComponents (Utilities|Struct|BreakSoundLibraryPlayResult _returnvalue_2))) "IsPawn" _bsuccess :!class /Script/AudioExtensions.AudioParameterControllerInterface)
            (return))
          (:Is%20Not%20Valid
            (Audio|Parameter|SetBooleanParameter (Utilities|Array|Get(acopy) (.AudioComponents (Utilities|Struct|BreakSoundLibraryPlayResult _returnvalue_2))) "Islocal" false :!class /Script/AudioExtensions.AudioParameterControllerInterface)
            (Audio|Parameter|SetBooleanParameter (Utilities|Array|Get(acopy) (.AudioComponents (Utilities|Struct|BreakSoundLibraryPlayResult _returnvalue_2))) "IsPawn" false :!class /Script/AudioExtensions.AudioParameterControllerInterface)))))))
